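# Animated material for the zone completion treasure chest, driving its motion and glow effects.
# Vertex offsets combine Pivot Painter lid rotation, height-masked tilt and squash-and-stretch, with rotation-corrected normals.
# Switchable, optionally animated emissive layers add to a height-lerped power swipe, Fresnel and noise-broken edge glow tinted by hue-shifted diffuse, then blend through whiteOut parameters.
# A base normal read from UV channels 1 and 2 is angle-blended with the normal map.
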Material /Game/Environments/Props/Boxes/Materials/Animated/AnimatedZoneCompletionChestMaterial
graph {
 properties {
  numCustomizedUVs=3
  pixelDepthOffsetMode=PDOM_Legacy
  previewMesh=/Game/Environments/Props/Boxes/Meshes/TreasureChestLootTier3
  usedShadingModels=Default Lit
 }
 nodes [
  n1 = ScalarParameter() {group=Vertex Shader, parameterName=Squash and stretch}
  n2 = Constant3Vector() {constant=<rgba 0 0 1 0>}
  n3 = WorldPosition() {worldPositionShaderOffset=WPT_ExcludeAllShaderOffsets}
  n4 = MaterialFunctionCall() {materialFunction=/Engine/Functions/Engine_MaterialFunctions02/PivotPainter/PivotPainter_PerObjectData}
  n5 = Multiply(A=@n4.Black Mask, B=ScalarParameter() {group=Vertex Shader, parameterName=ChestLidRotation})
  n6 = RotateAboutAxis(NormalizedRotationAxis=@n4.X-Axis Vector, RotationAngle=@n5, PivotPoint=@n4.X-Axis Vector, Position=@n3.XYZ)
  n7 = Normalize(VectorInput=Transform(None=Constant3Vector() {constant=<rgba 0 1 0 0>}) {transformSourceType=TRANSFORMSOURCE_Local})
  n8 = TransformPosition(None=WorldPosition() {worldPositionShaderOffset=WPT_ExcludeAllShaderOffsets}.XYZ) {transformSourceType=TRANSFORMPOSSOURCE_World}
  n9 = Clamp(None=Divide(A=MaterialFunctionCall(Float3=@n8) {materialFunction=&BreakOutFloat3Components}.B) {constB=30})
  n10 = Add(A=SphereMask(A=Constant(), B=@n8) {attenuationRadius=64, hardnessPercent=0}, B=Power(Base=OneMinus(None=@n9)) {constExponent=6})
  n11 = Multiply(A=ScalarParameter() {group=Vertex Shader, parameterName=Tilt}, B=OneMinus(None=Clamp(None=@n10)))
  n12 = Add(
   A=@n6
   B=RotateAboutAxis(
    NormalizedRotationAxis=@n7
    RotationAngle=@n11
    PivotPoint=MaterialFunctionCall() {materialFunction=&ComponentPivotLocation}.Result
    Position=Add(A=@n6, B=@n3.XYZ)
   )
  ) {desc=new deformaiton}
  n13 = TransformPosition(None=Add(A=@n3.XYZ, B=@n12) {desc=new wp}) {transformSourceType=TRANSFORMPOSSOURCE_World}
  n14 = Multiply(A=Multiply(A=@n1, B=Multiply(A=@n2, B=@n13)) {constB=-1}) {constB=-1}
  n15 = Multiply(A=@n1, B=Multiply(A=@n13, B=Constant3Vector() {desc=k, constant=<rgba 1 1 0 0>})) {constB=-1}
  Add(A=Transform(None=Add(A=@n14, B=@n15)) {transformSourceType=TRANSFORMSOURCE_Local}, B=@n12)
  n16 = VectorParameter() {group=Emissive, parameterName="EmissiveUVPositioning (RG)UpperLeft (BA)LowerRight", defaultValue=<rgba 0 0 1 1>}
  n17 = MaterialFunctionCall!(
   DiffuseFresnelAmount
   Emissive
   Base Color=TextureSampleParameter2D() {parameterName=Diffuse, texture=/Game/Environments/Props/Boxes/Textures/TreasureChestLV2_D}.RGB
  ) {desc=diffuse, materialFunction=/Game/Packages/Fortress_SharedMaterials/Base_Material_Functions/FresnelFunction}
  n18 = MaterialFunctionCall!(
   Manipulate Texture UVs=StaticBoolParameter() {group=Emissive, parameterName=Manipulate Emissive Uvs}
   ActorPositionAdd3
   Mask Cropped Emissive=StaticBoolParameter() {defaultValue=true, group=Emissive, parameterName="Mask Cropped Emissive "}
   Texture=TextureObjectParameter() {group=Emissive, parameterName=EmissiveTexture}
   Animated emissive black out time=ScalarParameter() {defaultValue=0.5, desc=should be a value from 0-1, group=Emissive, parameterName=Animated Emissive Blackout Ratio}
   Texture Multiplier=VectorParameter() {group=Emissive, parameterName=Emissive Multiplier, defaultValue=<rgba 1 1 1 1>}.RGB
   Modulate Texture=StaticBoolParameter() {group=Emissive, parameterName=Modulate Emissive}
   UVPositioning LowerRight=AppendVector(A=@n16.B, B=@n16.B)
   UVPositioning UpperLeft=ComponentMask(None=@n16.RGB) {g=true, r=true}
   "------------------- "
   "Animate Using Texture Based Animation(T) or a Flicker(F)"=StaticBoolParameter() {group=Emissive, parameterName=AnimateEmissiveUsingAGradient}
   Animate=StaticBoolParameter() {group=Emissive, parameterName=AnimateEmissive}
   -------------------
   "World Position Animation Offset Multiplier (WS)"=ScalarParameter() {defaultValue=0.01, group=Emissive, parameterName=Animated emissive world position animation offset amount}
   Use World Position to offset the animation=StaticBoolParameter() {group=Emissive, parameterName=Use World position to offset the emissive animation}
   Animated Radius=ScalarParameter() {defaultValue=0.1, group=Emissive, parameterName=Emissive Timeline Animation Falloff Radius}
   Animated Hardness=ScalarParameter() {defaultValue=0.1, group=Emissive, parameterName=Emissive Timeline Animation Falloff Hardness}
   "Loop(T) or Ping  Pong(F) Animation"=StaticBoolParameter() {group=Emissive, parameterName="Animate Emissive Using a loop(T) or Ping Pong(F)"}
   Modulate with diffuse=StaticBoolParameter() {group=Emissive, parameterName=Modulate Emissive with Diffuse}
   Diffuse=@n17.d
   ---------------------
   ControlTimeManually
   ManualTime
   Sequence animation speed=ScalarParameter() {defaultValue=0.25, group=Emissive, parameterName=EmissiveAnimationSpeed}
  ) {materialFunction=/Game/Packages/Fortress_SharedMaterials/Base_Material_Functions/SequenceAnimation}
  n19 = VectorParameter() {group=Emissive2, parameterName="Emissive 2 UV Positioning (RG)UpperLeft (BA)LowerRight", defaultValue=<rgba 0 0 1 1>}
  n20 = MaterialFunctionCall(
   Manipulate Texture UVs=StaticBoolParameter() {group=Emissive2, parameterName=Manipulate Emissive 2 Uvs}
   Mask Cropped Emissive=StaticBoolParameter() {defaultValue=true, group=Emissive2, parameterName=Mask Cropped Emissive 2}
   Texture=TextureObjectParameter() {group=Emissive2, parameterName=Emissive Texture 2}
   Animated emissive black out time=ScalarParameter() {defaultValue=0.5, desc=should be a value from 0-1, group=Emissive2, parameterName=Animated Emissive Blackout Ratio 2}
   Texture Multiplier=VectorParameter() {group=Emissive2, parameterName=Emissive Multiplier 2, defaultValue=<rgba 1 1 1 1>}.RGB
   Modulate Texture=StaticBoolParameter() {group=Emissive2, parameterName=Modulate Emissive 2}
   UVPositioning LowerRight=AppendVector(A=@n19.B, B=@n19.B)
   UVPositioning UpperLeft=ComponentMask(None=@n19.RGB) {g=true, r=true}
   "Animate Using Texture Based Animation(T) or a Flicker(F)"=StaticBoolParameter() {group=Emissive2, parameterName=Animate Emissive Using the emissive alpha as a timeline 2}
   Animate=StaticBoolParameter() {group=Emissive2, parameterName=Animate Emissive 2}
   "World Position Animation Offset Multiplier (WS)"=ScalarParameter() {defaultValue=0.01, group=Emissive2, parameterName=Animated emissive world position animation offset amount 2}
   Use World Position to offset the animation=StaticBoolParameter() {group=Emissive2, parameterName=Use World position to offset the emissive animation 2}
   Animated Radius=ScalarParameter() {defaultValue=0.1, group=Emissive2, parameterName=Emissive Timeline Animation Falloff Radius 2}
   Animated Hardness=ScalarParameter() {defaultValue=0.1, group=Emissive2, parameterName=Emissive Timeline Animation Falloff Hardness 2}
   "Loop(T) or Ping  Pong(F) Animation"=StaticBoolParameter() {group=Emissive2, parameterName="Animate Emissive Using a loop(T) or Ping Pong(F) 2"}
   Modulate with diffuse=StaticBoolParameter() {group=Emissive2, parameterName=Modulate Emissive with Diffuse 2}
   Diffuse=@n17.d
   Sequence animation speed=ScalarParameter() {defaultValue=0.25, group=Emissive2, parameterName=Emissive Animation Speed 2}
  ) {materialFunction=&SequenceAnimation}
  n21 = StaticSwitchParameter(
   True=StaticSwitchParameter(True=Add(A=@n18.Result, B=@n20.Result), False=@n18.Result) {group=Emissive2, parameterName=Use Emissive 2}
   False=Constant()
  ) {group=Emissive, parameterName=Use Emissive}
  n22 = StaticSwitchParameter(
   True=Multiply(A=MaterialFunctionCall!(In) {materialFunction=/Game/Packages/PBW/BASE/Materials/BASE_Animation}.BASE_EmissiveOutput, B=@n21)
   False=@n21
  ) {group=Special Cases, parameterName=Use BASE Emissive Animation And Levels}
  n23 = Fresnel(ExponentIn=ScalarParameter() {defaultValue=0.5, group=Emissive, parameterName=EmissveFresnelExponent})
  n24 = OneMinus(None=@n23)
  n25 = StaticSwitchParameter(True=Multiply(A=@n22, B=StaticSwitchParameter(True=@n24, False=@n23) {group=Emissive, parameterName=InvertEmissiveFresnel}), False=@n22) {group=Emissive, parameterName=Emissive Fresnel}
  n26 = ComponentMask(None=TextureCoordinate() {coordinateIndex=1}) {r=true}
  n27 = Multiply(A=ScalarParameter() {defaultValue=1, group=Emissive, parameterName=UVModulateSinePhase}, B=Time()) {constA=1}
  n28 = Multiply(
   A=@n25
   B=MaterialFunctionCall!(
    Sine Phase=Add(A=@n26, B=@n27)
    Value 1=ScalarParameter() {defaultValue=1, group=Emissive, parameterName=UVModulateSineValue1}
    Value 2=ScalarParameter() {group=Emissive, parameterName=UVModulateSineValue2}
   ) {materialFunction=/Engine/Functions/Engine_MaterialFunctions02/Utility/Sine_Remapped}.Remapped Sine
  )
  n29 = StaticBool() {value=true}
  n30 = Constant() {r=3}
  n31 = TextureSampleParameter2D!(UVs, Bias=Constant() {r=2}, Apply View MipBias) {
   desc=ao
   group=Swipe Gradient
   mipValueMode=TMVM_MipBias
   parameterName=Large scale AO map in Red Channel White Edge Map in Green
   samplerType=SAMPLERTYPE_LinearGrayscale
   texture=/Game/Environments/Props/Boxes/Textures/TreasureChestLootTier3_Searched001LightingMap
  }
  n32 = Multiply(
   A=Multiply(
    A=MaterialFunctionCall(Invert Fresnel=@n29, Power=@n30) {materialFunction=&Fresnel_Function}.Result
    B=ScalarParameter() {group=Swipe Gradient, parameterName=PowerFresnelIntensity}
   )
   B=@n31.R
  )
  n33 = ScalarParameter() {group=Swipe Gradient, parameterName=PowerSwipeAnim}
  n34 = MaterialFunctionCall(Float3=TransformPosition(None=WorldPosition() {worldPositionShaderOffset=WPT_ExcludeAllShaderOffsets}.XYZ) {transformSourceType=TRANSFORMPOSSOURCE_World}) {materialFunction=&BreakOutFloat3Components}
  n35 = Add(
   A=Divide(A=Add(A=@n34.R, B=@n34.R), B=VectorParameter() {group=Swipe Gradient, parameterName=Gradient Dimensions, defaultValue=<rgba 256 0 64 1>}.RGB) {constB=128}
   B=OneMinus(None=@n31.R)
  )
  n36 = Multiply(
   A=TextureSample(UVs=MaterialFunctionCall(Transition Phase=@n33, Height Texture=@n35) {materialFunction=&HeightLerp}.Alpha) {desc=Swipe Texture, texture=/Game/Environments/Props/Boxes/Textures/ZoneCompletionCrate}.RGB
   B=ScalarParameter() {defaultValue=1, group=Swipe Gradient, parameterName=PowerSwipeIntensity}
  )
  n37 = MaterialFunctionCall(
   Hue Shift Percentage=Constant() {r=0.5}
   Texture=TextureSampleParameter2D!(UVs, Apply View MipBias) {parameterName=Diffuse, texture=&TreasureChestLV2_D}.RGB
  ) {materialFunction=&HueShift}
  n38 = Multiply(A=Multiply(A=Add(A=@n32, B=@n36), B=@n37.Result)) {constB=5}
  n39 = Multiply(A=@n38, B=@n38) {constB=2}
  n40 = ScalarParameter() {group=Swipe Gradient, parameterName=PowerEdgeIntensity}
  n41 = TextureSampleParameter2D!(UVs, Bias, Apply View MipBias) {
   desc=ao
   group=Swipe Gradient
   mipValueMode=TMVM_MipBias
   parameterName=Large scale AO map in Red Channel White Edge Map in Green
   samplerType=SAMPLERTYPE_LinearGrayscale
   texture=&TreasureChestLootTier3_Searched001LightingMap
  }
  n42 = TextureSample(UVs=Panner(Coordinate=TextureCoordinate() {uTiling=0.5, vTiling=0.5}) {speedX=0.1, speedY=0.1}) {texture=/Engine/MapTemplates/TilingNoise05}
  n43 = Clamp(None=Subtract(A=@n41.G, B=Multiply(A=@n42.R) {constB=0.5})) {maxDefault=2}
  n44 = Power(Base=Multiply(A=@n37.Result, B=Multiply(A=@n40, B=@n43)))
  n45 = ScalarParameter() {parameterName=whiteOut, sliderMax=1}
  LinearInterpolate(
   A=Add(A=StaticSwitchParameter(True=@n28, False=@n25) {group=Emissive, parameterName=ModulateEmissiveByUVS}, B=Add(A=@n39, B=@n44))
   B=Multiply(A=Constant() {r=2}, B=@n45)
   Alpha=@n45
  )
  LinearInterpolate(A=@n17.d, B=Constant() {r=1}, Alpha=ScalarParameter() {group=Special Cases, parameterName=whiteOut, sliderMax=1})
  n46 = TextureCoordinate() {coordinateIndex=2}
  MaterialFunctionCall(
   BaseNormal=AppendVector(
    A=TextureCoordinate() {coordinateIndex=1}
    B=MaterialFunctionCall(Float2=@n46) {materialFunction=&BreakOutFloat2Components}.R
   )
   AdditionalNormal=TextureSampleParameter2D!(UVs, Apply View MipBias) {parameterName=Normal, samplerType=SAMPLERTYPE_Normal, texture=/Game/Environments/Props/Boxes/Textures/TreasureChestLV2_N}.RGB
  ) {materialFunction=&BlendAngleCorrectedNormals}
  MaterialFunctionCall!(
   UVs
   SpecRoughnessMin=ScalarParameter() {defaultValue=0.25, group=Specular, parameterName=SpecRoughnessMin}
   SpecRoughnessMax=ScalarParameter() {defaultValue=1, group=Specular, parameterName=SpecRoughnessMax}
   Metallic
  ) {materialFunction=/Game/Packages/Fortress_SharedMaterials/Base_Material_Functions/SpecularMasks}
  MaterialFunctionCall!(
   Rotation Axis=@n7
   Rotation Angle=@n11
   World Space Vertex Normals=MaterialFunctionCall(Rotation Axis=@n4.X-Axis Vector, Rotation Angle=@n5) {materialFunction=&FixRotateAboutAxisNormals}.New World Space Vertex Normal
   Movement Mask
  ) {materialFunction=/Engine/Functions/Engine_MaterialFunctions02/FixRotateAboutAxisNormals2}
  TextureSampleParameter2D!(UVs, Bias, Apply View MipBias) {
   desc=ao
   group=Swipe Gradient
   mipValueMode=TMVM_MipBias
   parameterName=AO map
   samplerType=SAMPLERTYPE_LinearGrayscale
   texture=&TreasureChestLootTier3_Searched001LightingMap
  }
  TextureSample() {texture=/Game/Environments/Props/Boxes/Textures/lowPolyEdgeMap}
 ]
}
